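; Targeting component on a MotoWrestle weapon that supplies air attack (pounce) trace parameters and tracks the latest trace hit.
; It finds the player pawn through the owning weapon and derives trace source and direction from camera-based targeting transforms, with distance from a hotfixable scalable float.
; Direction uses the camera transform directly to avoid extra calculations.
; Storing a hit result fires OnValidTargetInRangeChanged only when its blocking-hit state changes.
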(fn GetAirAttackTraceDistance
  ;[0.09233 0.02717 0.125] Returns the air attack trace distance.
  (return (ScalableFloat|GetValueatLevel (Variables|Default|GetHotfix_PounceTargetingTraceDistance))))

(fn GetLatestTargetingTraceHitResult
  ; Returns the latest targeting trace hit result.
  (return (Variables|Transient|GetLatestTargetingTraceHitResult)))

(fn GetPlayerPawn
  ;[0.1192 0.01785 0.1406] Returns the owning fort player pawn for the targeting component.
  (Utilities|FlowControl|Sequence
    (:then_0
      (bind _asfort_weapon (Utilities|Casting|CastToFortWeapon (Components|GetOwner)))
      (bind _asfort_player_pawn (Utilities|Casting|CastToFortPlayerPawn (Actor|GetOwner _asfort_weapon)))
      (return _asfort_player_pawn))
    (:then_1
      (return 0))))

(fn GetTargetingSourceLocation
  ;[0.06709 0.02381 0.1042] GetAirAttackSourceLocation: Returns the location of the targeting transform.
  (bind _transform (Targeting|GetTargetingTransform))
  (bind _transform_location (Math|Transform|BreakTransform _transform))
  (return _transform_location))

(fn GetTargetingTraceDirection
  (bind _playerpawn (CallFunction|GetPlayerPawn))
  (bind (_returnvalue_location _returnvalue_rotation) (Math|Transform|BreakTransform (Targeting|GetTargetingTransform _playerpawn "Camera" 0 :!class /Script/FortniteGame.FortPawn)))
  ;[0.1074 0.02336 0.1406] Returns the trace direction for the air attack. Uses camera transform instead of calculating targeting transform to avoid unessasary calculations.
  (Utilities|FlowControl|Sequence
    (:then_0
      (Utilities|IsValid _playerpawn
        (:Is%20Valid
          (return (Math|Vector|GetForwardVector _returnvalue_rotation)))))
    (:then_1
      (return "0, 0, 0"))))

(fn GetTargetingTransform
  (bind _playerpawn (CallFunction|GetPlayerPawn))
  ;[0.1458 0.03458 0.09946] GetAirAttackTargetingTransform: If successful, the returned transform will be a point on the Player's camera forward vector that is nearest the player pawn.
  (Utilities|FlowControl|Sequence
    (:then_0
      (Utilities|IsValid _playerpawn
        (:Is%20Valid
          (bind _returnvalue (TargetingSystem|GetPlayerCameraTargetingTransform _playerpawn true))
          (return _returnvalue))))
    (:then_1
      (return 0))))

(fn IsValidTargetInRange
  ; Returns if there's a valid blocking hit in the latest targeting trace hit result.
  (return (Collision|BreakHitResult (Variables|Transient|GetLatestTargetingTraceHitResult))))

(fn SetLatestTargetingTraceHitResult
  (params (HitResult :type Hit%20Result%20Structure))
  ; SetLatestTargetingTraceHitResult: Sets the latest trace hit result from the targeting GAs.
  (Utilities|FlowControl|Sequence
    (:then_0
      (LocalVariables|SetbValidTargetInRangeChanged (!= (Collision|BreakHitResult HitResult) (Collision|BreakHitResult (Variables|Transient|GetLatestTargetingTraceHitResult)))))
    (:then_1
      (Variables|Transient|SetLatestTargetingTraceHitResult HitResult)
      (if (LocalVariables|GetbValidTargetInRangeChanged)
        (Default|CallOnValidTargetinRangeChanged (CallFunction|IsValidTargetinRange))))))
