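; B_MotoWrestle_Weapon is a weapon Blueprint for the MotoWrestle gameplay feature.
; It overrides the out-of-ammo text getter to return a localized "NO CHARGES" message, so empty-weapon feedback refers to charges.
; Its construction script only calls the parent, and the file shows no other logic.
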(fn GetOutOfAmmoTextOverride
  (return "NSLOCTEXT(\"[8B8CB54BC26E792C33638DBE2AA2F911]\", \"91092BA045565C5F494298BA58DBF20C\", \"NO CHARGES\")"))

(fn ConstructionScript
  (|CallParentFunction))
